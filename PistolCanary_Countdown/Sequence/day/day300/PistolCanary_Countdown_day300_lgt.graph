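# Lighting Level Sequence for the day300 shot of PistolCanary_Countdown, driving one spotlight and its light function.
# It possesses the PistolCanary_Countdown SpotLight and sets its light component intensity to a fixed value.
# A material parameter collection track on MPC_PIstolCanary_LightFunction holds constant, opposite-signed speeds for two textures, presumably scrolling the light function pattern.
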LevelSequence /PistolCanary_Countdown/Sequence/day/day300/PistolCanary_Countdown_day300_lgt
bindings [
 LightComponent0 {
  id=#1
  parent_id=#2
  possessed_class=SpotLightComponent
  tracks [{
    type=MovieSceneFloatTrack
    display_name=Intensity
    sections [{type=MovieSceneFloatSection, channels [None {type=MovieSceneScriptingFloatChannel, default=10000}]}]
   }]
 }
 SpotLight_PistolCanary_Countdown {id=#2, possessed_class=SpotLight, tracks []}
]
display_rate {denominator=1, numerator=60}
tick_resolution {denominator=1, numerator=24000}
tracks [{
  type=MovieSceneMaterialParameterCollectionTrack
  display_name=MPC_PIstolCanary_LightFunction
  sections [{
    type=MovieSceneParameterSection
    channels [
     Texture1 - Speed {type=MovieSceneScriptingFloatChannel, default=-120, const [-120,1]}
     Texture2 - Speed {type=MovieSceneScriptingFloatChannel, default=113, const [113,1]}
    ]
   }]
 }]
